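# Subsurface material for the animated, glowing bubbly liquid in the Medic pickaxe, usable on skeletal meshes.
# Fresnel and panning smoke blend dark and light colors, then shift toward green along the UV's V axis.
# Three panning bubble-normal layers brighten that color and distort a fake view-space reflection, while ripple-warped plasma adds flowing highlights.
# The combined emissive is scaled by time of day, dimmer at night, through a shared FX adjustment function.
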Material /BRCosmetics/Weapons/FORT_Melee/Pickaxe_Medic/Materials/M_PickAxeMedic_Liquid
graph {
 properties {
  bUsedWithMeshDeformer=true
  bUsedWithSkeletalMesh=true
  pixelDepthOffsetMode=PDOM_Legacy
  shadingModel=MSM_Subsurface
  usedShadingModels=Subsurface
 }
 nodes [
  n1 = VectorParameter() {parameterName=ColorDark, defaultValue=<rgba 0 0.1407 0.25 1>}
  n2 = VectorParameter() {parameterName=ColorLight, defaultValue=<rgba 0.2 1.4667 4 1>}
  n3 = MaterialFunctionCall(Exponent=ScalarParameter() {defaultValue=10, parameterName="Exponent (S)"}) {materialFunction=&AxisAlignedFresnel}
  n4 = TextureCoordinate()
  n5 = ComponentMask(None=VectorParameter() {parameterName=ColorVari_UV_Scale, defaultValue=<rgba 2 0.25 0 0>}.RGB) {g=true, r=true}
  n6 = Time()
  n7 = ScalarParameter() {defaultValue=0.45, parameterName=Time Multi_2}
  n8 = TextureSample(UVs=Panner(Coordinate=Multiply(A=@n4, B=@n5), Time=Multiply(A=@n6, B=@n7)) {speedY=0.25}) {texture=&T_Inky_Smoke_Tile}
  n9 = Multiply(A=Saturate(None=@n3.Fresnel), B=@n8.G)
  n10 = VectorParameter() {parameterName=ColorCore_Dark, defaultValue=<rgba 0 0.1882 0.5 1>}
  n11 = Saturate(None=Power(Base=OneMinus(None=@n3.Fresnel)) {constExponent=4})
  n12 = ComponentMask(None=TextureCoordinate() {vTiling=1.5}) {g=true}
  n13 = ScalarParameter() {defaultValue=1, parameterName=ColorSwtich_Exp}
  n14 = LinearInterpolate(
   A=Multiply(A=LinearInterpolate(A=@n1.RGB, B=@n2.RGB, Alpha=@n9), B=LinearInterpolate(A=@n10.RGB, B=@n2.RGB, Alpha=@n11))
   B=VectorParameter() {parameterName=Green_Liquid_Color, defaultValue=<rgba 0.009 0.7 0.2958 1>}.RGB
   Alpha=Clamp(None=Power(Base=@n12, Exp=@n13))
  )
  Desaturation(None=@n14, Fraction=Multiply(A=@n8.G, B=ScalarParameter() {defaultValue=3, parameterName=Subsurface_Desat}))
  n15 = ScalarParameter() {defaultValue=0.35, parameterName=EmissiveBoost}
  n16 = ScalarParameter() {defaultValue=0.15, parameterName=EmissiveBoost_Night}
  n17 = Transform(None=ReflectionVectorWS()) {transformSourceType=TRANSFORMSOURCE_World, transformType=TRANSFORM_View}
  n18 = Multiply(A=ConstantBiasScale(None=ComponentMask(None=@n17) {g=true, r=true}), B=Constant2Vector() {g=-1, r=1})
  n19 = TextureCoordinate()
  n20 = ComponentMask(None=VectorParameter() {parameterName=Bubble_Tile_Scale_1, defaultValue=<rgba 5 5 0 0>}.RGB) {g=true, r=true}
  n21 = TextureSample(UVs=Panner(Coordinate=Multiply(A=@n19, B=@n20)) {speedX=-0.05, speedY=0.123}) {
   samplerType=SAMPLERTYPE_Normal
   texture=/Game/Characters/Player/Male/Medium/Bodies/M_Med_Commando_AlienFishHead/Textures/T_SimpleBubbleNormal__N
  }
  n22 = Add(A=TextureCoordinate(), B=VectorParameter() {parameterName=Bubble_UV_Scale_3, defaultValue=<rgba 0 0.01 0 0>}.G)
  n23 = ComponentMask(None=VectorParameter() {parameterName=Bubble_Tile_Scale_2, defaultValue=<rgba 4 4 0 0>}.RGB) {g=true, r=true}
  n24 = TextureSample(UVs=Panner(Coordinate=Multiply(A=@n22, B=@n23)) {speedX=0.01, speedY=0.14}) {
   samplerType=SAMPLERTYPE_Normal
   texture=&T_SimpleBubbleNormal__N
  }
  n25 = Add(A=TextureCoordinate(), B=VectorParameter() {parameterName=Bubble_UV_Scale_4, defaultValue=<rgba 0.5 0.01 0 0>}.G)
  n26 = ComponentMask(None=VectorParameter() {parameterName=Bubble_Tile_Scale_3, defaultValue=<rgba 3 3 0 0>}.RGB) {g=true, r=true}
  n27 = TextureSample(UVs=Panner(Coordinate=Multiply(A=@n25, B=@n26)) {speedX=0.02, speedY=0.2}) {
   samplerType=SAMPLERTYPE_Normal
   texture=&T_SimpleBubbleNormal__N
  }
  n28 = Power(Base=Add(A=Add(A=@n21.RGB, B=@n24.RGB), B=@n27.RGB), Exp=ScalarParameter() {defaultValue=2, parameterName=BubblePowerMult}) {constExponent=1}
  n29 = Multiply(A=OneMinus(None=Desaturation(None=@n28)), B=ScalarParameter() {defaultValue=1, parameterName=NormalMult})
  n30 = Multiply(A=ScalarParameter() {defaultValue=0.8, parameterName=ReflectDistort_Mult}, B=@n29)
  n31 = Multiply(
   A=TextureSample(UVs=Add(A=@n18, B=@n30)) {texture=/Game/Environments/Props/Rocks/Textures/T_Props_CrystalEnv_White}.RGB
   B=VectorParameter() {parameterName=ReflectionColor, defaultValue=<rgba 0 0 3 1>}.RGB
  )
  n32 = TextureCoordinate()
  n33 = ComponentMask(None=VectorParameter() {parameterName=Liquid_UV_Scale, defaultValue=<rgba 0.5 2 0 0>}.RGB) {g=true, r=true}
  n34 = TextureSample(UVs=@n32) {samplerType=SAMPLERTYPE_Normal, texture=/Game/Effects/Fort_Effects/Gameplay/Water/Textures/T_WaterRipples_01}
  n35 = Panner(
   Coordinate=Add(A=Multiply(A=@n32, B=@n33), B=ComponentMask(None=@n34.RGB) {g=true, r=true})
   Time=Multiply(A=Time(), B=ScalarParameter() {defaultValue=0.5, parameterName=Time Multi_1})
  ) {speedY=0.1}
  n36 = Clamp(None=Power(
    Base=TextureSample(UVs=@n35) {texture=/Game/Effects/Fort_Effects/Textures/Tile/T_Plasma_BallCore}.G
    Exp=ScalarParameter() {defaultValue=1, parameterName=Contrast}
   ))
  n37 = Multiply(A=@n14, B=Clamp(None=@n29) {maxDefault=2})
  n38 = Power(Base=Add(A=Multiply(A=@n36, B=@n14), B=Add(A=@n14, B=@n37)), Exp=ScalarParameter() {defaultValue=1.25, parameterName=EmissivePower})
  MaterialFunctionCall(Emissive=Multiply(
    A=MaterialFunctionCall(Color_Morning=@n15, Color_Day=@n15, Color_Evening=@n15, Color_Night=@n16) {materialFunction=&Color_Time_of_Day}.Result
    B=Add(A=@n31, B=@n38)
   )) {materialFunction=&MF_FX_EmissiveAdjustments}
  Normalize(VectorInput=@n28)
  OneMinus(None=@n29)
  ScalarParameter() {defaultValue=1, parameterName=Specular_1}
  ScalarParameter() {defaultValue=1, parameterName=Metallic_1}
 ]
}
