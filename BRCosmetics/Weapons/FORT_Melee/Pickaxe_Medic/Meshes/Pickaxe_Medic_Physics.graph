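# Physics asset for the Medic pickaxe, a two-handed melee cosmetic, setting up physics for its dangling bag and glove.
# Capsule bodies cover the handle and each dyn_ bone, with limited angular constraints forming two branches off the handle.
# The bag swings widely and chains to a tight bottom joint and a loose tube.
# The glove moves within narrower limits and carries four finger bodies on small, uniform limits.
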PhysicsAsset /BRCosmetics/Weapons/FORT_Melee/Pickaxe_Medic/Meshes/Pickaxe_Medic_Physics
bodies {
 Melee_TwoHanded_Handle [{shapeType=Capsule, length=110.4354, radius=2.95, center=<0 -1.561 26.1169>}]
 dyn_bag [{shapeType=Capsule, length=9.75, radius=6.25, center=<0 0 -7.5495>, rotation=<rot 0 -90 0>}]
 dyn_bag_bottom [{shapeType=Capsule, length=1.755, radius=2.755}]
 dyn_bag_tube [{shapeType=Capsule, length=1.0474, radius=1.75}]
 dyn_glove [{shapeType=Capsule, length=6.936, radius=3.8614, center=<0.0002 0.1881 -6.5089>}]
 dyn_index [{shapeType=Capsule, length=2.9067, radius=1.2749, center=<1.0615 -2.2375 -2.8915>, rotation=<rot 36.0804 -90 -15.8253>}]
 dyn_middle [{shapeType=Capsule, length=3.8169, radius=1.5733, center=<0.4068 -0.5015 -3.2047>, rotation=<rot 8.894 -90 -7.0093>}]
 dyn_pinky [{shapeType=Capsule, length=2.7445, radius=1.1656, center=<0.4428 1.2332 -1.854>, rotation=<rot -33.631 -90 -10.5579>}]
 dyn_ring [{shapeType=Capsule, length=3.6505, radius=1.6401, center=<0.2597 0.5113 -3.216>, rotation=<rot -9.0568 -90 -4.4138>}]
}
constraints [
 {bone1Name=dyn_bag, bone2Name=Melee_TwoHanded_Handle, swing1LimitDegrees=100.2857, swing2LimitDegrees=83.2857, twistLimitDegrees=112.4286}
 {bone1Name=dyn_bag_bottom, bone2Name=dyn_bag, swing1LimitDegrees=15, swing2LimitDegrees=15, twistLimitDegrees=15}
 {bone1Name=dyn_glove, bone2Name=Melee_TwoHanded_Handle, swing1LimitDegrees=33, swing2LimitDegrees=31.2857, twistLimitDegrees=9}
 {bone1Name=dyn_index, bone2Name=dyn_glove, swing1LimitDegrees=20, swing2LimitDegrees=20, twistLimitDegrees=20}
 {bone1Name=dyn_middle, bone2Name=dyn_glove, swing1LimitDegrees=20, swing2LimitDegrees=20, twistLimitDegrees=20}
 {bone1Name=dyn_ring, bone2Name=dyn_glove, swing1LimitDegrees=20, swing2LimitDegrees=20, twistLimitDegrees=20}
 {bone1Name=dyn_pinky, bone2Name=dyn_glove, swing1LimitDegrees=20, swing2LimitDegrees=20, twistLimitDegrees=20}
 {bone1Name=dyn_bag_tube, bone2Name=dyn_bag_bottom, swing1LimitDegrees=80, swing2LimitDegrees=75, twistLimitDegrees=115}
]
