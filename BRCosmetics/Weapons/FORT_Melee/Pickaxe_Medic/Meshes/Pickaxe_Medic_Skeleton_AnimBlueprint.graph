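# Animation Blueprint for the Pickaxe_Medic melee weapon skeleton that adds physics-driven secondary motion.
# The input pose passes through one RigidBody node using Pickaxe_Medic_Physics, with the two-handed handle as base bone, overridden gravity and clamped component acceleration, then goes to output.
# An AnimDynamics node and two additive bone rotations for the bag bones sit unconnected.
# The event graph has no logic.
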AnimBlueprint /BRCosmetics/Weapons/FORT_Melee/Pickaxe_Medic/Meshes/Pickaxe_Medic_Skeleton_AnimBlueprint
target_skeleton=&/Pickaxe_Medic_Skeleton
graphs [
 AnimGraph {nodes [
   Anim Dynamics - Bone: dyn_bag(Alpha=1, ExternalForce="0,0,0") {
    bGravityOverrideInSimSpace=true
    alphaBoolBlend {blendOption=Linear}
    boundBone {boneName=dyn_bag}
    physicsBodyDefinitions [{
      sphereCollisionRadius=0
      boundBone {boneName=dyn_bag}
      boxExtents=<40 40 40>
      constraintSetup {twistAxis=AxisY, angularLimitsMax=<45 0 0>, angularLimitsMin=<-45 0 0>, angularTarget=<0 0 0>}
      localJointOffset=<0 -1.5 -20>
     }]
   }
   "Transform (Modify) Bone - Bone: dyn_bag"(Alpha=1, Translation="0,0,0", Rotation="0,0,-30", Scale="1,1,1") {rotationMode=BMM_Additive, alphaBoolBlend {blendOption=Linear}, boneToModify {boneName=dyn_bag}, rotation {roll=-30}}
   "Transform (Modify) Bone - Bone: dyn_bag_bottom"(Alpha=1, Translation="0,0,0", Rotation="0,0,-60", Scale="1,1,1") {rotationMode=BMM_Additive, alphaBoolBlend {blendOption=Linear}, boneToModify {boneName=dyn_bag_bottom}, rotation {roll=-60}}
   AnimGraphNode_SubInput_0 = Input Pose()
   n1 = RigidBody(ComponentPose=Local To Component(LocalPose=@AnimGraphNode_SubInput_0)) {
    overridePhysicsAsset=&/Pickaxe_Medic_Physics
    alphaBoolBlend {blendOption=Linear}
    baseBoneRef {boneName=Melee_TwoHanded_Handle}
    componentAppliedLinearAccClamp=<1000 1000 1000>
    componentLinearAccScale=<1 1 1>
    overrideWorldGravity=<0.8 0.8 0.8>
   }
   Output Pose(Result=Component To Local(ComponentPose=@n1)) {layerGroup=None}
  ]}
 EventGraph {nodes [Event BlueprintUpdateAnimation()]}
]
variables []
